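# Sound cue that plays the Fortnite Reload Hotel music track as a looping gameplay music bed.
# A single looping wave player feeds the FN_Reload_Hotel_Music_Loop sound wave straight to the output.
# It is routed to the Fort_Music_Gameplay sound class and uses the Fort_Ambient_Athena_Att attenuation, with a very large max distance.
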SoundCue /StarsoapGameplay/Assets/Audio/SC_FN_Reload_Hotel_Music_Loop_Cue
root=WavePlayer_0
nodes [{id=WavePlayer_0, properties {bLooping=true, soundWaveAssetPtr=&/FN_Reload_Hotel_Music_Loop}}]
properties {
 attenuationSettings=&Fort_Ambient_Athena_Att
 maxDistance=2097152
 soundClassObject=/Game/Sounds/FortSoundClassesAndModes/Fort_Music_Gameplay
}
