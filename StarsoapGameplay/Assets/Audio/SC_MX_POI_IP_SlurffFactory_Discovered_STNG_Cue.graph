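# Sound cue for the discovery stinger of the Slurff Factory point of interest, routed to the music sound class.
# A mixer layers three one-shots: a magical POI-discovered UI sound, a party-invite UI sound and the Slurff Factory music stinger.
# Each layer passes through its own modulator with equal min and max volume, so levels are fixed rather than randomized.
# It uses a musical one-shot attenuation asset without reverb.
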SoundCue /StarsoapGameplay/Assets/Audio/SC_MX_POI_IP_SlurffFactory_Discovered_STNG_Cue
root=Mixer_0
nodes [
 {id=Mixer_0, children [Modulator_0,Modulator_2,Modulator_1], properties {inputVolume [0.9,1.2,0.75]}}
 {id=Modulator_0, children [WavePlayer_1], properties {volumeMax=1.5, volumeMin=1.5}}
 {id=Modulator_1, children [WavePlayer_0], properties {pitchMax=1, pitchMin=1, volumeMax=1.8, volumeMin=1.8}}
 {id=Modulator_2, children [WavePlayer_2], properties {volumeMax=1, volumeMin=1}}
 {id=WavePlayer_0, properties {soundWaveAssetPtr=&/MX_S38_POI_IP_SlurffFactory_Discovered_STNG}}
 {id=WavePlayer_1, properties {soundWaveAssetPtr=/Game/Sounds/Fort_UI_Sounds/Gameplay/InGame_POI_Discovered_02_Magical}}
 {id=WavePlayer_2, properties {soundWaveAssetPtr=/Game/Sounds/Fort_UI_Sounds/FrontEnd/ItemShop/wav/UI_PartyInvite_Appear}}
]
properties {
 attenuationSettings=/Game/Sounds/Attenuation/Fort_Ambient_Athena_Musical_Oneshot_NoVerb_Att
 maxDistance=2097152
 soundClassObject=/Game/Sounds/FortSoundClassesAndModes/Fort_Music
}
